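# Editor Utility Widget that serves as the art-tool panel for the Biome Brush Manager actor.
# On Construct, it shows a "No BiomeBrushManager Found!" notice, soft-loads the manager class and searches for an actor of that class.
# A valid actor becomes the Details View's object, replacing the widget itself, and hides the notice.
# The Details View lists only Configuration, Actions and Debug categories, exposing the manager's properties without custom controls.
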EditorUtilityWidgetBlueprint /Game/NeverCook/ArtTools/Tech_Art/BiomeBrushManager/Widgets/EUW_BiomeBrushManager
graphs [EventGraph {nodes [
   Event Construct()
   > SetObject(self=Get DetailsView_22() -> (DetailsView_22:Details View Object Reference), NewObject=Self-Reference())
   > SetVisibility(self=Get TextBlock_82() -> (TextBlock_82:Text Object Reference), InVisibility=Visible)
   > n1 = LoadClassAsset_Blocking(AssetClass=Get TargetClass() -> (TargetClass:EUBP Biome Brush Manager Soft Class Reference))
   > n2 = Cast To Actor Class!(execute:Exec, Class:Object Class Reference=@n1) -> (then:Exec, CastFailed:Exec, AsActor:Actor Class Reference) {is=title K2Node_ClassDynamicCast}
   > n3 = GetActorOfClass!(execute, WorldContextObject:Object Reference, ActorClass:Actor Class Reference=@n2) -> (then, ReturnValue:Actor Object Reference)
   > Is Valid(InputObject=@n3)
   > SetObject(self=Get DetailsView_22(), NewObject=@n3)
   > SetVisibility(self=Get TextBlock_82(), InVisibility=Collapsed)
  ]}]
variables [TargetClass {type=Class, default=/Game/NeverCook/ArtTools/Tech_Art/BiomeBrushManager/EUBP_BiomeBrushManager.EUBP_BiomeBrushManager_C}]
widgets {
 info {rootWidgetClass=&CanvasPanel, widgetCount=4}
 widgets [
  CanvasPanel_0 {widgetClassPath=&CanvasPanel, widget=CanvasPanel_0}
  Overlay_0 {
   widgetClassPath=&Overlay
   parent=CanvasPanel_0
   widget=Overlay_0
   slot {class=CanvasPanelSlot, properties {layoutData {alignment=<0 0>, anchors {maximum=<1 1>, minimum=<0 0>}, offsets=<ltrb 0 0 0 0>}}}
  }
  DetailsView_22 {
   widgetClassPath=&DetailsView
   parent=Overlay_0
   widget=DetailsView_22
   properties {categoriesToShow [Configuration, Actions, Debug]}
   slot {class=OverlaySlot, properties {horizontalAlignment=HAlign_Fill, verticalAlignment=VAlign_Fill}}
  }
  TextBlock_82 {
   widgetClassPath=&TextBlock
   bIsVariable=true
   parent=Overlay_0
   widget=TextBlock_82
   properties {
    text="No BiomeBrushManager Found!"
    colorAndOpacity {colorUseRule=UseColor_Specified, specifiedColor=<rgba 1 1 1 0.5>}
    font {
     bForceMonospaced=false
     bMaterialIsStencil=false
     fontMaterial=null
     fontObject=&Roboto
     letterSpacing=0
     monospacedWidth=1
     size=9
     skewAmount=0
     typefaceFontName=Regular
     outlineSettings {
      bApplyOutlineToDropShadows=false
      bMiteredCorners=false
      bSeparateFillAlpha=false
      outlineMaterial=null
      outlineSize=0
      outlineColor=<rgba 0 0 0 1>
     }
    }
   }
   slot {class=OverlaySlot, properties {horizontalAlignment=HAlign_Center, verticalAlignment=VAlign_Center}}
  }
 ]
}
